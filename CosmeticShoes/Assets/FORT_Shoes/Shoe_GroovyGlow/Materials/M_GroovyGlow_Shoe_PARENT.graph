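# Parent character material for the GroovyGlow cosmetic shoe, with fluffy fabric and an audio-reactive emissive glow.
# Character PreFX feeds fluffy fabric (using a UV-stepped tech art mask) and shoe reactivity functions, then the emissive override, sock function and PostFX.
# The glow is a time-based sine, phased by the mask's green channel and audio amplitude, with a cosine-driven fake frontend audio loop.
# Amplitude blends dim emissive toward the pulse, gated by ReactivityOn.
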Material /CosmeticShoes/Assets/FORT_Shoes/Shoe_GroovyGlow/Materials/M_GroovyGlow_Shoe_PARENT
graph {
 properties {
  bUseMaterialAttributes=true
  bUsedWithClothing=true
  bUsedWithMeshDeformer=true
  bUsedWithMorphTargets=true
  bUsedWithNanite=true
  bUsedWithNiagaraMeshParticles=true
  bUsedWithSkeletalMesh=true
  pixelDepthOffsetMode=PDOM_Legacy
  shadingModel=MSM_FromMaterialExpression
  usedShadingModels=Default Lit | Subsurface | Clear Coat | Subsurface Profile | Cloth
 }
 nodes [
  n1 = MaterialFunctionCall() {materialFunction=&MF_CharacterParent_2023_PreFX}
  n2 = TextureCoordinate()
  n3 = Multiply(A=@n1.Tech Art Mask, B=Step(Y=ComponentMask(None=@n2) {r=true}))
  n4 = GetMaterialAttributes(None=MaterialFunctionCall!(Material=MaterialFunctionCall!(Material=@n1.Result, M=@n1.Result, TechArtMask=@n3) {materialFunction=/BRCosmetics/Accessories/FORT_Backpacks/Backpack_M_MED_PearlTote/Materials/MF_FluffyFabric}.Result) {materialFunction=/Game/Global/Materials/Cosmetics/Kicks/Shoes/Materials/MF_ShoeReactivityTest}.Result) {attributeGetTypes [#1]}
  n5 = Constant() {r=0.01}
  n6 = ComponentMask(None=@n1.Tech Art Mask) {g=true}
  n7 = Multiply(A=Time()) {constB=0.8}
  n8 = Cosine(None=Divide(A=Time(), B=ScalarParameter() {defaultValue=1.3, group=Audio Reactivity, parameterName=FrontendFakeAudio_LoopLength, sortPriority=200}))
  n9 = MaterialFunctionCall!(
   InAmplitude
   InSpectrum
   CustomFakeSpectrum
   CustomFakeAmplitude=Multiply(
    A=ConstantBiasScale(None=@n8)
    B=ScalarParameter() {defaultValue=0.5, group=Audio Reactivity, parameterName=FrontendFakeAudio_Amplitude, sortPriority=200}
   )
  ) {materialFunction=/BRCosmetics/Global/MaterialFunctions/AudioReactivity/MF_AudioReactivity_FrontendLoop}
  n10 = Multiply(A=OneMinus(None=Power(Base=@n9.Amplitude) {constExponent=6})) {constB=0.4}
  n11 = Sine(None=Add(A=Multiply(A=@n6), B=Subtract(A=@n7, B=@n10)))
  n12 = LinearInterpolate(A=@n5, Alpha=Add(A=Multiply(A=@n11) {constB=0.5}) {constB=0.5}) {constA=0.05, constB=1.5}
  n13 = Multiply(A=@n9.Amplitude)
  n14 = LinearInterpolate(
   A=Multiply(A=@n4.Emissive Color, B=@n5)
   B=LinearInterpolate(A=Multiply(A=@n4.Emissive Color, B=@n5), B=Multiply(A=@n4.Emissive Color, B=@n12), Alpha=Saturate(None=@n13))
   Alpha=ScalarParameter() {defaultValue=1, group=GroovyGlow, parameterName=ReactivityOn}
  )
  MaterialFunctionCall(
   Material=MaterialFunctionCall!(
    InOpacity
    Material=SetMaterialAttributes(MaterialAttributes=@n4.MaterialAttributes, Emissive Color=@n14) {attributeSetTypes [#1]}
    Mask=@n1.Tech Art Mask
   ) {materialFunction=/Game/Global/Materials/Cosmetics/Kicks/Socks/Materials/MF_Sock}.Result
   M=@n1.M
   Pre Skinned Local Position=@n1.M
   Distance Blend=@n1.M
   Fresnel=@n1.M
  ) {materialFunction=&MF_CharacterParent_2023_PostFX}
 ]
}
